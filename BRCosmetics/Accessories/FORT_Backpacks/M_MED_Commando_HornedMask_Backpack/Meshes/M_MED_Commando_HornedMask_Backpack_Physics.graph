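# Physics asset for the Commando Horned Mask backpack cosmetic, providing collision bodies and joints for its dangling mask, beard, mustache and tassels.
# A capsule and box form the backpack root, which anchors the mask, with tight swing and looser twist, and two three-box tassel chains.
# The mask carries three two-sphere beard chains and two mustache capsules on limited joints.
# Tassel root and tip joints lock twist, while middle joints swing freely.
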PhysicsAsset /BRCosmetics/Accessories/FORT_Backpacks/M_MED_Commando_HornedMask_Backpack/Meshes/M_MED_Commando_HornedMask_Backpack_Physics
bodies {
 backpack_root [
  {shapeType=Capsule, length=13.7932, radius=4.5677, center=<-9.1183 -6.7439 0.48>}
  {shapeType=Box, extentX=20.5995, extentY=7.5, extentZ=19.1379, center=<4.6524 -7.086 0.0002>}
 ]
 dyn_beard_01 [{shapeType=Sphere, radius=1.6649}]
 dyn_beard_01_l [{shapeType=Sphere, radius=1.6649}]
 dyn_beard_01_r [{shapeType=Sphere, radius=1.6649}]
 dyn_beard_02 [{shapeType=Sphere, radius=1.6649}]
 dyn_beard_02_l [{shapeType=Sphere, radius=1.6649, center=<0.4679 0 0.0001>}]
 dyn_beard_02_r [{shapeType=Sphere, radius=1.6649, center=<0.4202 0 0>}]
 dyn_mask [{shapeType=Capsule, length=13.5964, radius=5.0594, center=<-0.2833 -2.5165 -8.9509>, rotation=<rot 0 0 -4.606>}]
 dyn_stache_l [{shapeType=Capsule, length=4.5574, radius=2.8875, center=<-5.6334 1.8425 -0.935>, rotation=<rot -5.7065 -104.9279 -96.2232>}]
 dyn_stache_r [{shapeType=Capsule, length=4.5574, radius=2.8875, center=<5.6315 1.9109 -0.9089>, rotation=<rot -6.4956 -75.7194 -80.3928>}]
 dyn_tassela_01 [{shapeType=Box, extentX=1.8204, extentY=1.0358, extentZ=3.7522, center=<1.8499 -0.0139 0.0922>, rotation=<rot 0 90 90>}]
 dyn_tassela_02 [{shapeType=Box, extentX=1.6804, extentY=0.9396, extentZ=3.6312, center=<1.7426 -0.0729 -0.0433>, rotation=<rot 0 -92.3955 -88.5807>}]
 dyn_tassela_03 [{shapeType=Box, extentX=2.5775, extentY=2.1713, extentZ=6.6332, center=<3.2363 -0.0335 -0.0726>, rotation=<rot 0 -93.1622 -85.4012>}]
 dyn_tasselb_01 [{shapeType=Box, extentX=1.9962, extentY=0.8181, extentZ=3.6853, center=<1.9958 0 -0.0001>, rotation=<rot 0 -90 90>}]
 dyn_tasselb_02 [{shapeType=Box, extentX=2.7801, extentY=0.8195, extentZ=1.7645, center=<1.5922 0.0017 0.0003>, rotation=<rot 0 177.857 -89.8209>}]
 dyn_tasselb_03 [{shapeType=Box, extentX=2.909, extentY=2.264, extentZ=6.1915, center=<3.058 -0.2784 -0.1722>, rotation=<rot 0 -91.2996 -89.7521>}]
}
constraints [
 {bone1Name=dyn_tassela_03, bone2Name=dyn_tassela_02, swing1LimitDegrees=20, swing2LimitDegrees=20, twistLimitDegrees=45, twistMotion=Locked}
 {bone1Name=dyn_tasselb_03, bone2Name=dyn_tasselb_02, swing1LimitDegrees=20, swing2LimitDegrees=20, twistLimitDegrees=45, twistMotion=Locked}
 {bone1Name=dyn_beard_02_r, bone2Name=dyn_beard_01_r, swing1LimitDegrees=30, swing2LimitDegrees=30, twistLimitDegrees=5}
 {bone1Name=dyn_beard_02_l, bone2Name=dyn_beard_01_l, swing1LimitDegrees=30, swing2LimitDegrees=30, twistLimitDegrees=5}
 {bone1Name=dyn_beard_02, bone2Name=dyn_beard_01, swing1LimitDegrees=30, swing2LimitDegrees=30, twistLimitDegrees=5}
 {bone1Name=dyn_beard_01_r, bone2Name=dyn_mask, swing1LimitDegrees=15, swing2LimitDegrees=30, twistLimitDegrees=2}
 {bone1Name=dyn_beard_01_l, bone2Name=dyn_mask, swing1LimitDegrees=15, swing2LimitDegrees=30, twistLimitDegrees=2}
 {bone1Name=dyn_beard_01, bone2Name=dyn_mask, swing1LimitDegrees=15, swing2LimitDegrees=30, twistLimitDegrees=2}
 {bone1Name=dyn_stache_r, bone2Name=dyn_mask, swing1LimitDegrees=20, swing2LimitDegrees=20, twistLimitDegrees=15}
 {bone1Name=dyn_stache_l, bone2Name=dyn_mask, swing1LimitDegrees=20, swing2LimitDegrees=20, twistLimitDegrees=15}
 {
  bone1Name=dyn_tasselb_02
  bone2Name=dyn_tasselb_01
  swing1LimitDegrees=45
  swing1Motion=Free
  swing2LimitDegrees=45
  swing2Motion=Free
  twistLimitDegrees=10
 }
 {
  bone1Name=dyn_tassela_02
  bone2Name=dyn_tassela_01
  swing1LimitDegrees=45
  swing1Motion=Free
  swing2LimitDegrees=45
  swing2Motion=Free
  twistLimitDegrees=10
 }
 {bone1Name=dyn_tassela_01, bone2Name=backpack_root, swing1LimitDegrees=10, swing2LimitDegrees=10, twistLimitDegrees=45, twistMotion=Locked}
 {bone1Name=dyn_tasselb_01, bone2Name=backpack_root, swing1LimitDegrees=10, swing2LimitDegrees=10, twistLimitDegrees=45, twistMotion=Locked}
 {bone1Name=dyn_mask, bone2Name=backpack_root, swing1LimitDegrees=2, swing2LimitDegrees=2, twistLimitDegrees=45}
]
