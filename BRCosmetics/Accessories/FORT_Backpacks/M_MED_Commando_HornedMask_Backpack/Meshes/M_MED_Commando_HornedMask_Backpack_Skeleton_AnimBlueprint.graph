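# Animation Blueprint for the horned mask Commando back bling, adding physics-driven secondary motion to the backpack skeleton.
# The input pose runs through one RigidBody node using the backpack's physics asset, simulated relative to backpack_root with gravity override and scaled component motion.
# Anim Dynamics nodes for the two moustache bones and a second RigidBody node sit disconnected from the output.
# The event graph sets CrouchReverse from inverted bIsCrouching, but the shown AnimGraph never reads it.
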AnimBlueprint /BRCosmetics/Accessories/FORT_Backpacks/M_MED_Commando_HornedMask_Backpack/Meshes/M_MED_Commando_HornedMask_Backpack_Skeleton_AnimBlueprint
target_skeleton=&/M_MED_Commando_HornedMask_Backpack_Skeleton
graphs [
 AnimGraph {nodes [
   n1 = Anim Dynamics - Bone: dyn_stache_r(Alpha=1, ExternalForce="0,0,0") {
    angularDampingOverride=0.95
    angularSpringConstant=800
    bAngularSpring=true
    bEnableWind=true
    bGravityOverrideInSimSpace=true
    bOverrideAngularDamping=true
    bUseGravityOverride=true
    simulationSpace=BoneRelative
    alphaBoolBlend {blendOption=Linear}
    boundBone {boneName=dyn_stache_r}
    componentAppliedLinearAccClamp=<50 50 50>
    componentLinearAccScale=<0.6 0.6 0.6>
    componentLinearVelScale=<0.45 0.45 0.45>
    gravityOverride {z=-980}
    physicsBodyDefinitions [{
      sphereCollisionRadius=0
      boundBone {boneName=dyn_stache_r}
      boxExtents=<10 5 5>
      constraintSetup {angularLimitsMax=<2 10 5>, angularLimitsMin=<-2 -10 -5>, angularTarget=<0 0 0>}
      localJointOffset=<-0.1 0 -1>
     }]
    relativeSpaceBone {boneName=dyn_mask}
   }
   Anim Dynamics - Bone: dyn_stache_l(ComponentPose=@n1, Alpha=1, ExternalForce="0,0,0") {
    angularDampingOverride=0.95
    angularSpringConstant=800
    bAngularSpring=true
    bEnableWind=true
    bGravityOverrideInSimSpace=true
    bOverrideAngularDamping=true
    bUseGravityOverride=true
    simulationSpace=BoneRelative
    alphaBoolBlend {blendOption=Linear}
    boundBone {boneName=dyn_stache_l}
    componentAppliedLinearAccClamp=<50 50 50>
    componentLinearAccScale=<0.6 0.6 0.6>
    componentLinearVelScale=<0.45 0.45 0.45>
    gravityOverride {z=-980}
    physicsBodyDefinitions [{
      sphereCollisionRadius=0
      boundBone {boneName=dyn_stache_l}
      boxExtents=<10 5 5>
      constraintSetup {angularLimitsMax=<2 10 5>, angularLimitsMin=<-2 -10 -5>, angularTarget=<0 0 0>}
      localJointOffset=<0.1 0 -1>
     }]
    relativeSpaceBone {boneName=dyn_mask}
   }
   RigidBody() {alphaBoolBlend {blendOption=Linear}}
   AnimGraphNode_SubInput_0 = Input Pose()
   n2 = RigidBody(ComponentPose=Local To Component(LocalPose=@AnimGraphNode_SubInput_0)) {
    bOverrideWorldGravity=true
    overridePhysicsAsset=&/M_MED_Commando_HornedMask_Backpack_Physics
    simulationSpace=BaseBoneSpace
    alphaBoolBlend {blendOption=Linear}
    baseBoneRef {boneName=backpack_root}
    componentAppliedLinearAccClamp=<1500 1500 1500>
    componentLinearAccScale=<0.4 0.4 0.4>
    componentLinearVelScale=<0.5 0.5 0.5>
    overrideWorldGravity {z=-980}
   }
   Output Pose(Result=Component To Local(ComponentPose=@n2)) {layerGroup=None}
  ]}
 EventGraph {nodes [
   Event BlueprintUpdateAnimation()
   n1 = Get bIsCrouching()
   > Set CrouchReverse(CrouchReverse:double="To Float (Boolean)"(InBool=NOT Boolean(A=@n1))) -> (Output_Get:double)
  ]}
]
variables [CrouchReverse {type=number}]
